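; Parent building game actor for a world Sprite that players can pick up or that turns into an AI after a timer.
; With uses left, the server spawns AI through a spawner utility component or Living World Manager, retrying failures; otherwise the actor silently dies.
; Tracking ownership passes to the equipped held weapon or spawner-created pawn.
; Hits attach it to moving platforms, and a temporary server-only hotfix repeatedly forces simulation stops.
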(event Custom|ForceStop
  (bind _servercachedhit (Variables|Default|GetServerCachedHit))
  (bind (_bblockinghit _binitialoverlap _time _distance _location _impactpoint _normal _impactnormal _physmat _hitactor _hitcomponent _hitbonename _bonename _hititem _elementindex _faceindex _tracestart _traceend) (Collision|BreakHitResult _servercachedhit))
  (bind _outvalue (Fortnite|GetNumericCVar "Sprites.ServerBGA.RandomImpactLocationMagnitude"))
  (bind _returnvalue (Math|Random|RandomUnitVector))
  (bind _returnvalue_2 (* _returnvalue _outvalue))
  (bind _returnvalue_3 (+ _impactpoint _returnvalue_2))
  (bind _returnvalue_4 (Collision|MakeHitResult _bblockinghit _binitialoverlap _time _distance _location _returnvalue_3 _normal _impactnormal _physmat _hitactor _hitcomponent _hitbonename _bonename _hititem _elementindex _faceindex _tracestart _traceend))
  (bind (_bblockinghit_2 _binitialoverlap_2 _time_2 _distance_2 _location_2) (Collision|BreakHitResult (Variables|Default|GetServerCachedHit)))
  ; Server Only 33.00 Fix - Will be removed in 33.30
  (if (== (Variables|Default|GetServerForceStopCount) 0)
    (Networking|SetNetDormancy "DORM_Awake")
    (Networking|FlushNetDormancy)
    (Variables|Default|SetbResumeSimulation true)
    (Math|Integer|IncrementInt (Variables|Default|GetServerForceStopCount))
    (else
      (Networking|SetReplicateMovement true)
      (Transformation|SetActorLocation _location_2)
      (Math|Integer|IncrementInt (Variables|Default|GetServerForceStopCount)
        (:%20%20
          (Networking|SetNetDormancy "DORM_Awake")
          (Networking|FlushNetDormancy)
          (Variables|Default|SetBlockStoppingSim false)
          (Variables|Default|SetbResumeSimulation false)
          (Variables|Default|SetStopSimServerLocation (Transformation|GetActorLocation self))
          (Variables|Default|SetStopSimHitResult _returnvalue_4)
          (CallFunction|StopSim _returnvalue_4)
          (if (>= (Variables|Default|GetServerForceStopCount) (Math|Float|Floor (Fortnite|GetNumericCVar "Sprites.ServerBGA.ForceStopMessageCount")))
            (Utilities|Time|ClearandInvalidateTimerbyHandle (Variables|Default|GetServerForceStopTimer))))))))

(event Custom|StartForceStop
  ; Server Only 33.00 Fix - Will be removed in 33.30
  (Networking|SwitchHasAuthority
    (:Authority
      (if (> (Fortnite|GetNumericCVar "Fort.Rollback.SpritesServerBGAHotfix") 0)
        (bind _returnvalue (Utilities|Time|SetTimerbyEvent (EventDispatchers|CreateEvent :!event Custom|ForceStop) (Fortnite|GetNumericCVar "Sprites.ServerBGA.ForceStopInterval") true true))
        (Variables|Default|SetServerForceStopTimer _returnvalue)))))

(event EventStopSim
  (params (Hit :type Hit%20Result%20Structure))
  ; Server Only 33.00 Fix - Will be removed in 33.30
  (|CallParentFunction Hit)
  (Networking|SwitchHasAuthority
    (:Authority
      (Variables|Default|SetServerCachedHit Hit)
      (CallFunction|StartForceStop))))

(event EventEndPlay
  (params (EndPlayReason :type EEndPlayReason%20Enum :default "Destroyed"))
  (|CallParentFunction EndPlayReason)
  (Networking|SwitchHasAuthority
    (:Authority
      (Class|FortAthenaLivingWorldManager|UnregisterDensityReservoir (Class|FortAthenaLivingWorldManager|GetLivingWorldManager) self)
      (bind _currentinteractingpawn (Variables|Default|GetCurrentInteractingPawn)
        (:then
          (FortPawn|UnbindEventfromOnWeaponEquippedDelegate _currentinteractingpawn (EventDispatchers|CreateEvent :!event OnSpriteWeaponEquipped)))
        (:else)))))

(event EventBeginPlay
  (|CallParentFunction)
  (GameplayCue|AddGameplayCueOnActor(Looping) self (Variables|Default|GetGameplay%20Cue%20Tag) (Ability|GameplayCue|MakeGameplayCueParameters 0.0 0.0 (GameplayEffects|MakeEffectContext (Ability|GetAbilitySystemComponent 0))))
  (Utilities|FlowControl|DelayUntilNextTick)
  (if (not (Variables|Default|GetIsSleeping))
    (Utilities|Time|SetTimerbyEvent (EventDispatchers|CreateEvent :!event Custom|SpawnAI) (CallFunction|GetSpawnAIDelay)))
  (CallFunction|ActivateSpriteAbility)
  (if (Class|BPSpritePickupComponent|HasRemainingUses (Variables|Default|GetBP_SpritePickupComponent))
    (Networking|SwitchHasAuthority
      (:Authority
        (Variables|Default|SetEnableInteraction true)))))

(event Interaction|EventBlueprintOnInteract
  (params (InteractingPawn :type Fort%20Pawn%20Object%20Reference) (InteractionBeingAttempted :type EInteractionBeingAttempted%20Enum :default "FirstInteraction"))
  (bind _asfort_player_pawn (Utilities|Casting|CastToFortPlayerPawn InteractingPawn))
  (bind _spritegiven (Class|BPSpritePickupComponent|PickupSprite (Variables|Default|GetBP_SpritePickupComponent) (Fortnite|GetFortPlayerControllerfromActor _asfort_player_pawn)))
  (if _spritegiven
    (bind _output_get (Variables|Default|SetCurrentInteractingPawn InteractingPawn))
    (FortPawn|BindEventtoOnWeaponEquippedDelegate _output_get (EventDispatchers|CreateEvent :!event OnSpriteWeaponEquipped))))

(event Custom|WakeUpSprite
  (CallFunction|SpawnAI))

(event Custom|ActivateSpriteAbility
  (if (CallFunction|CanActivateAbility)
    (CallFunction|ExecuteSpriteAbility)))

(event Custom|ExecuteSpriteAbility)

(event Custom|SpawnAI
  (Networking|SwitchHasAuthority
    (:Authority
      (if (Class|BPSpritePickupComponent|HasRemainingUses (Variables|Default|GetBP_SpritePickupComponent))
        (bind _returnvalue (CallFunction|SpawnAIInternal (delegate Custom|OnAISpawned)))
        (if _returnvalue
          (Rendering|SetVisibility (Variables|Default|GetStaticMesh) false false)
          (Gameplay|SilentDie false)
          (else
            (CallFunction|HandleFailedSpawn)))
        (else
          (CallFunction|OutOfCharges)
          (Rendering|SetVisibility (Variables|Default|GetStaticMesh) false false)
          (Default|CallOnSpriteActivationsExhausted (Variables|Default|GetBP_SpriteTrackingComponent))
          (Gameplay|SilentDie false))))))

(event Custom|OnAISpawned
  (params (SpawnedActors :type Array%20of%20Actor%20Object%20References) (bSuccess :type Boolean))
  (if bSuccess
    (for _array_element SpawnedActors
      (Class|SpritesUtilityLibrary|SpritesTryPropagateItemEntrytoNextActor self _array_element))
    (else
      (CallFunction|HandleFailedSpawn))))

(event Custom|EnterSleepState
  (bind _returnvalue (Actor|GetComponentbyClass "/Script/FortniteGame.FortItemEntryComponent"))
  (Utilities|FlowControl|Delay (ScalableFloat|GetCachedScalableFloatValue (Variables|Default|GetHF_TimeOnTrainNoCharges)))
  (Utilities|IsValid _returnvalue
    (:Is%20Valid
      (if (Class|FortItemEntryComponent|HasItem _returnvalue)
        (if (== (Class|FortItemEntryComponent|GetLoadedAmmo _returnvalue) 0)
          (CallFunction|OutOfCharges)
          (Rendering|SetVisibility (Variables|Default|GetStaticMesh) false false)
          (Default|CallOnSpriteActivationsExhausted (Variables|Default|GetBP_SpriteTrackingComponent))
          (Gameplay|SilentDie false))))))

(event Custom|HandleFailedSpawn
  (if (not (Variables|Default|GetIsSleeping))
    (Variables|Default|SetIsSleeping true))
  (Utilities|FlowControl|RetriggerableDelay (ScalableFloat|GetValueatLevel (Variables|Default|GetRetrySpawnAIDelay)))
  (CallFunction|SpawnAI))

(event Custom|OutOfCharges)

(event Collision|EventHit
  (params (MyComp :type Primitive%20Component%20Object%20Reference) (Other :type Actor%20Object%20Reference) (OtherComp :type Primitive%20Component%20Object%20Reference) (bSelfMoved :type Boolean) (HitLocation :type Vector :default "0, 0, 0") (HitNormal :type Vector :default "0, 0, 0") (NormalImpulse :type Vector :default "0, 0, 0") (Hit :type Hit%20Result%20Structure%20%28by%20ref%29))
  (bind (_bblockinghit _binitialoverlap _time _distance _location _impactpoint _normal _impactnormal _physmat _hitactor) (Collision|BreakHitResult Hit))
  (bind _returnvalue (Class|FortBaseMovingPlatformComponent|FindMovingPlatformComponentfromActor _hitactor true))
  (bind _returnvalue_2 (Components|GetOwner _returnvalue))
  (|CallParentFunction MyComp Other OtherComp bSelfMoved HitLocation HitNormal NormalImpulse Hit)
  (Networking|SwitchHasAuthority
    (:Authority
      _)
    (:Remote
      _))
  (Utilities|IsValid _returnvalue
    (:Is%20Valid
      (Utilities|IsValid _returnvalue_2
        (:Is%20Valid
          (Utilities|FlowControl|DoOnce false
            (:Completed
              (bind _returnvalue_3 (Transformation|AttachActorToActor self _returnvalue_2 "None" "KeepWorld" "KeepWorld" "KeepWorld" true))
              (if _returnvalue_3
                _
                (else
                  _)))))))))

(event OnLinkedActorDestroyed(FortLinkToActor)
  (params (DamageCauser :type Actor%20Object%20Reference))
  (CallFunction|RestartSimulation))

(event OnLinkedActorDisconnected(FortLinkToActor)
  (CallFunction|RestartSimulation))

(fn BlueprintCanInteract
  (params (InteractingPawn :type Fort%20Pawn%20Object%20Reference) (InteractionBeingAttempted :type EInteractionBeingAttempted%20Enum :default "FirstInteraction") (InteractionType :type TInteractionType%20Enum :default "IT_NoInteraction"))
  (if (Variables|Default|GetEnableInteraction)
    (return (Class|BPSpritePickupComponent|CanPickup (Variables|Default|GetBP_SpritePickupComponent) InteractingPawn))
    (else
      (return false))))

(fn BlueprintGetInteractionString
  (params (InteractingPawn :type Fort%20Pawn%20Object%20Reference) (InteractionBeingAttempted :type EInteractionBeingAttempted%20Enum :default "FirstInteraction"))
  (return (Variables|GetBuildingActorInteractionText)))

(fn BlueprintGetInteractionTime
  (params (InteractingPawn :type Fort%20Pawn%20Object%20Reference) (InteractionBeingAttempted :type EInteractionBeingAttempted%20Enum :default "FirstInteraction"))
  (if (CallFunction|CanActivateAbility)
    (return (ScalableFloat|GetValueatLevel (Variables|Default|GetSlowInteractionTime)) true)
    (else
      (return (ScalableFloat|GetValueatLevel (Variables|Default|GetFastInteractionTime)) true))))

(fn CanActivateAbility
  (return (Variables|Default|GetbExecuteSpriteAbility)))

(fn GetSpawnAIDelay
  (if (Class|BPSpritePickupComponent|HasRemainingUses (Variables|Default|GetBP_SpritePickupComponent))
    (return (ScalableFloat|GetValueatLevel (Variables|Default|GetHF_AISpawnTimer)))
    (else
      (return (ScalableFloat|GetValueatLevel (Variables|Default|GetHF_AISpawnTimerNoCharges))))))

(fn GetSpriteType
  (return "NewEnumerator0"))

(fn HasRemainingUses
  (return false))

(fn OnAthenaAISpawnerRequestComplete
  (params (Pawn :type Pawn%20Object%20Reference) (RequestID :type Integer))
  (if (== RequestID (Variables|Default|GetAISpawnRequestID))
    (EventDispatchers|UnbindEventfromOnPawnSpawned (Variables|Default|GetAthenaAISpawner) (EventDispatchers|CreateEvent :!event OnAthenaAISpawnerRequestComplete))
    (Variables|Default|SetAthenaAISpawner 0)
    (Variables|Default|SetAISpawnRequestID -1)
    (CallFunction|OnAISpawned (Utilities|Array|MakeArray (Variables|GetPawn)) true)
    (Class|BPSpriteTrackingComponent|TransferTrackingOwnership (Variables|Default|GetBP_SpriteTrackingComponent) (Actor|GetComponentbyClass (Variables|GetPawn) "/Sprites/Gameplay/Components/BP_SpriteTrackingComponent.BP_SpriteTrackingComponent_C"))))

(fn OnRep_DisableInteraction)

(fn OnRep_IsSleeping
  (CallFunction|EnterSleepState))

(fn OnSpriteWeaponEquipped
  (params (NewWeapon :type Fort%20Weapon%20Object%20Reference) (PrevWeapon :type Fort%20Weapon%20Object%20Reference))
  (bind _asb_sprite_held_weapon_parent (Utilities|Casting|CastToB_Sprite_Held_Weapon_Parent NewWeapon))
  (Class|BPSpriteTrackingComponent|TransferTrackingOwnership (Variables|Default|GetBP_SpriteTrackingComponent) (Actor|GetComponentbyClass _asb_sprite_held_weapon_parent "/Sprites/Gameplay/Components/BP_SpriteTrackingComponent.BP_SpriteTrackingComponent_C"))
  (bind _currentinteractingpawn (Variables|Default|GetCurrentInteractingPawn)
    (:then
      (FortPawn|UnbindEventfromOnWeaponEquippedDelegate _currentinteractingpawn (EventDispatchers|CreateEvent :!event OnSpriteWeaponEquipped))
      (Variables|Default|SetCurrentInteractingPawn 0))
    (:else
      _))
  (Gameplay|SilentDie false))

(fn SpawnAI_Internal
  (params (OnRequestFinished :type Delegate%20%28by%20ref%29))
  (bind _outputdelegate (EventDispatchers|CreateEvent :!event OnAthenaAISpawnerRequestComplete))
  (if (and (Utilities|IsValid (Variables|/Sprites/Gameplay/Components/BP_SpriteTrackingComponent.BP_SpriteTrackingComponent_C|GetAISpawnerUtilityComp (Variables|Default|GetBP_SpriteTrackingComponent)) :!class /Script/Engine.KismetSystemLibrary) (Utilities|IsValid (Variables|/Sprites/Gameplay/Components/BP_SpriteTrackingComponent.BP_SpriteTrackingComponent_C|GetAISpawnerDataComponentList (Variables|Default|GetBP_SpriteTrackingComponent)) :!class /Script/Engine.KismetSystemLibrary))
    ; FNE Spawn Path - uses the AthenaAISpawner vias the spawn utility component passed from the Wildlife Spawner device. FNE does not use the LWM
    (bind _returnvalue (Spawning|RequestSpawn (Variables|/Sprites/Gameplay/Components/BP_SpriteTrackingComponent.BP_SpriteTrackingComponent_C|GetAISpawnerUtilityComp (Variables|Default|GetBP_SpriteTrackingComponent)) (Variables|/Sprites/Gameplay/Components/BP_SpriteTrackingComponent.BP_SpriteTrackingComponent_C|GetAISpawnerDataComponentList (Variables|Default|GetBP_SpriteTrackingComponent)) (Math|Transform|MakeTransform (Transformation|GetActorLocation) (Transformation|GetActorRotation)) -1))
    (if (not (== _returnvalue -1))
      (Variables|Default|SetAISpawnRequestID _returnvalue)
      (bind _returnvalue_2 (AI|GetAISpawner))
      (Utilities|IsValid _returnvalue_2
        (:Is%20Valid
          (bind _output_get (Variables|Default|SetAthenaAISpawner _returnvalue_2))
          (EventDispatchers|UnbindEventfromOnPawnSpawned _output_get _outputdelegate)
          (EventDispatchers|BindEventtoOnPawnSpawned _output_get _outputdelegate))))
    (return (not (== _returnvalue -1)))
    (else
      ; BR Spawn Path
      (bind (_requestid _returnvalue_3) (Class|FortAthenaLivingWorldManager|TrySpawnEventforRegistry (Class|FortAthenaLivingWorldManager|GetLivingWorldManager) (Utilities|Struct|MakeDataRegistryId "SpriteThrownLagerEvents" (Variables|Default|GetLWM%20Event%20Name)) (Transformation|GetActorTransform) (Variables|GetOnRequestFinished)))
      (Variables|Default|SetAISpawnRequestID _requestid)
      (return _returnvalue_3))))

(fn ConstructionScript
  (|CallParentFunction))
